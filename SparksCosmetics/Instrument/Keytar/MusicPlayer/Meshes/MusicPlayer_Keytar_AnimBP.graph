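# Animation Blueprint for the MusicPlayer keytar instrument mesh in SparksCosmetics, layering an idle animation onto an incoming pose.
# The AnimGraph uses Apply Additive to add the MusicPlayer_Keytar_Idle_Loop sequence on top of the Input Pose.
# Rotator variables for a barrel and four cogs are declared but never read in the graphs shown, and the update event is empty.
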AnimBlueprint /SparksCosmetics/Instrument/Keytar/MusicPlayer/Meshes/MusicPlayer_Keytar_AnimBP
target_skeleton=&/MusicPlayer_Keytar_Skeleton
graphs [
 AnimGraph {nodes [Output Pose(Result=Apply Additive(
     Base=Input Pose()
     Additive="Sequence Player (additive) 'MusicPlayer_Keytar_Idle_Loop'"() {sequence=&/MusicPlayer_Keytar_Idle_Loop}
    ) {alphaBoolBlend {blendOption=Linear}}) {layerGroup=None}]}
 EventGraph {nodes [Event BlueprintUpdateAnimation()]}
]
variables [
 Barrel Rotator {type=Rotator, default=<rot 0 90 0>}
 Lower_Cog_L_Rotator {type=Rotator}
 LowerCog_R_Rotator {type=Rotator}
 Upper_Cog_L_Rotator {type=Rotator}
 Upper_Cog_R_Rotator {type=Rotator}
]
